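# Niagara system for the MeowKnaw pickaxe skin's weapon trail, drawn as short-lived ribbons.
# One emitter spawns two anim trails, each spanning its own pair of bone or socket indices, with a user RibbonIDOffset parameter exposed.
# Masks choose between two warm spawn colors and between two RGB scale curves indexed by Fortnite time of day.
# Anim-trail update modules drive the particles while an alpha curve fades them out.
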NiagaraSystem /BRCosmetics/Weapons/FORT_Melee/Pickaxe_Spooky_Neon_Male/Skins/MeowKnaw/FX/NS_Pickaxe_MeowKnaw_Trail
dependencies {
 usedDataInterfaces [&NiagaraDataInterfaceColorCurve, &NiagaraDataInterfaceCurve, &NiagaraDataInterfaceSocketReader]
 usedDynamicInputs []
 usedModules [
  &self:SystemSpawnScript.SystemScriptSource.SystemScriptGraph.NiagaraNodeAssignment_0.SetVariables_#1
  &SystemState
 ]
}
emitters {Main_Trail {
  emitterSpawnScript [&FN_RibbonIDOffsetUtility()]
  emitterUpdateScript [
   &EmitterState()
   &FN_SpawnAnimTrail(Number of Anim Trails=2)
  ]
  particleSpawnScript [
   &FN_InitializeAnimTrailBoneIndecies(Second Bone or Socket Index=2)
   &FN_InitializeAnimTrailBoneIndecies(AnimTrail Selection=1, First Bone or Socket Index=1, Second Bone or Socket Index=2) {module=FN_InitializeAnimTrailBoneIndecies001}
   &FN_InitializeAnimTrail(
    Color=&MaskLinearColor(Mask Source=0, Mask Parameter=?, Mask=0, Masked Linear Color=<rgba 1 0.5579 0 0.6>, Passthrough Linear Color=<rgba 0.4793 0.0762 0.011 0.9>)
    Lifetime=0.27
   )
  ]
  particleUpdateScript [
   &SolveForcesAndVelocity(Clamp Velocity=false, Limit Acceleration=false)
   &FN_UpdateAnimTrail() {module=FN_UpdateAnimTrail001}
   &ScaleColor(
    Scale Alpha=&FloatFromCurve(FloatCurve=DataInterface() {propertyValues {Curve {
        defaultValue=1
        keys [
         {arriveTangentWeight=0.0167, interpMode=RCIM_Cubic, leaveTangentWeight=0.0917, tangentMode=RCTM_User, time=0, value=1}
         {arriveTangent=-10.779, arriveTangentWeight=1.5933, interpMode=RCIM_Cubic, leaveTangent=-10.779, tangentMode=RCTM_User, time=0.325, value=0}
        ]
       }}})
    Scale RGB=&MaskVector(
     Mask Source=0
     Mask Parameter=?
     Mask=0
     Masked Vector=&MakeVectorFromLinearColorRGB(Color=&ColorFromCurve(
       ColorCurve=DataInterface() {propertyValues {
         AlphaCurve {keys [<at 0 1 RCIM_Cubic>]}
         BlueCurve {keys [<at 0 0.65 RCIM_Cubic>,<at 0.5 0.75>,<at 1 0.65 RCIM_Cubic>]}
         GreenCurve {keys [<at 0 1.3 RCIM_Cubic>,<at 0.5 1.5>,<at 1 1.3 RCIM_Cubic>]}
         RedCurve {keys [<at 0 1.3 RCIM_Cubic>,<at 0.5 1.5>,<at 1 1.3 RCIM_Cubic>]}
        }}
       CurveIndex=Linked() {linkedVariable {name=NPC.FortniteNPC.TimeOfDay_0-1, type=NiagaraFloat}}
      ))
     Passthrough Vector=&MakeVectorFromLinearColorRGB(Color=&ColorFromCurve(
       ColorCurve=DataInterface() {propertyValues {
         AlphaCurve {keys [<at 0 1>]}
         BlueCurve {keys [<at 0 0.7>,<at 0.5 1>,<at 1 0.7>]}
         GreenCurve {keys [<at 0 0.5679>,<at 0.5 1>,<at 1 0.5679>]}
         RedCurve {keys [<at 0 0.42>,<at 0.5 1>,<at 1 0.42>]}
        }}
       CurveIndex=Linked() {linkedVariable {name=NPC.FortniteNPC.TimeOfDay_0-1, type=NiagaraFloat}}
      ))
    )
    ScaleA=true
    ScaleRGB=true
   )
  ]
  rendererClasses [&NiagaraRibbonRendererProperties]
  renderers [{rendererClass=&NiagaraRibbonRendererProperties}]
 }}
summary {
 systemName=NS_Pickaxe_MeowKnaw_Trail
 emitters [{emitterName=Main_Trail, rendererClasses [&NiagaraRibbonRendererProperties]}]
 userVariables [User.RibbonIDOffset {type=NiagaraInt32}]
}
